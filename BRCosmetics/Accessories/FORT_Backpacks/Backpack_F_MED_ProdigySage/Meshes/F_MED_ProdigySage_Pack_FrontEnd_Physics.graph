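# Physics asset for the FrontEnd variant of the ProdigySage backpack, giving its chain, ears and spikes bodies for secondary motion.
# Capsule bodies hang off the pack through small base anchors whose constraints leave rotation free.
# Dynamic bones have twist locked and limited swing: the chain links run in series, while both segments of each ear constrain directly to that ear's base.
# Spikes swing on one axis only, with differing limits.
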PhysicsAsset /BRCosmetics/Accessories/FORT_Backpacks/Backpack_F_MED_ProdigySage/Meshes/F_MED_ProdigySage_Pack_FrontEnd_Physics
bodies {
 base_chain [{shapeType=Capsule, length=0.505, radius=0.505, shapeName=base_chain_capsule, rotation=<rot 0 -90 0>}]
 base_ear_l [{shapeType=Capsule, length=0.505, radius=0.505, shapeName=base_ear_l_capsule, rotation=<rot 0 -90 0>}]
 base_ear_r [{shapeType=Capsule, length=0.505, radius=0.505, shapeName=base_ear_r_capsule, rotation=<rot 0 -90 0>}]
 base_spike_01 [{shapeType=Capsule, length=0.505, radius=0.505, shapeName=base_spike_01_capsule, rotation=<rot 0 -90 0>}]
 base_spike_02 [{shapeType=Capsule, length=0.505, radius=0.505, shapeName=base_spike_02_capsule, rotation=<rot 0 -90 0>}]
 dyn_chain_01 [{shapeType=Capsule, length=0.505, radius=0.505, shapeName=dyn_chain_01_capsule, center=<0.7261 0 0>, rotation=<rot -90 180 -180>}]
 dyn_chain_02 [{shapeType=Capsule, length=0.505, radius=0.505, shapeName=dyn_chain_02_capsule, center=<0.731 0 0>, rotation=<rot -90 -2.0249 2.0249>}]
 dyn_chain_03 [{shapeType=Capsule, length=3.5, radius=1.5, shapeName=dyn_chain_03_capsule, center=<4.315 -0.6272 0>, rotation=<rot -90 -179.8945 179.8946>}]
 dyn_ear_l_01 [{shapeType=Capsule, length=3, radius=1, shapeName=dyn_ear_l_01_capsule, rotation=<rot 0 -90.0013 -30>}]
 dyn_ear_l_02 [{
   shapeType=Capsule
   length=3.894
   radius=1.3926
   shapeName=dyn_ear_l_02_capsule
   center=<2.5826 0.2544 -0.3932>
   rotation=<rot 0 -91.0171 -90.6138>
  }]
 dyn_ear_r_01 [{shapeType=Capsule, length=3, radius=1, shapeName=dyn_ear_l_01_capsule, rotation=<rot 0 -90.0014 30>}]
 dyn_ear_r_02 [{
   shapeType=Capsule
   length=3.894
   radius=1.3926
   shapeName=dyn_ear_l_02_capsule
   center=<2.5717 0.2546 0.6276>
   rotation=<rot 0 -91.0171 -90.6138>
  }]
 dyn_spike_01 [{shapeType=Capsule, length=10, radius=1.5183, shapeName=dyn_spike_01_capsule, center=<0 0 0.4619>, rotation=<rot -90 180 178.3288>}]
 dyn_spike_02 [{shapeType=Capsule, length=10, radius=1.5183, shapeName=dyn_spike_01_capsule, center=<0.2481 0 0.4619>, rotation=<rot -90 180 176.3953>}]
 pack [
  {shapeType=Capsule, length=10.5, radius=7.4296, shapeName=pack_capsule, center=<0 -6.0892 0>, rotation=<rot -50 0.0002 -0.0003>}
  {shapeType=Capsule, length=6.5, radius=9.5, shapeName=pack_capsule, center=<-6.4654 9.2548 -8.9246>, rotation=<rot 90 0 0>}
 ]
}
constraints [
 {
  bone1Name=base_chain
  bone2Name=pack
  swing1LimitDegrees=45
  swing1Motion=Free
  swing2LimitDegrees=45
  swing2Motion=Free
  twistLimitDegrees=45
  twistMotion=Free
 }
 {bone1Name=dyn_chain_01, bone2Name=base_chain, swing1LimitDegrees=45, swing2LimitDegrees=45, twistLimitDegrees=45, twistMotion=Locked}
 {bone1Name=dyn_chain_02, bone2Name=dyn_chain_01, swing1LimitDegrees=45, swing2LimitDegrees=45, twistLimitDegrees=45, twistMotion=Locked}
 {bone1Name=dyn_chain_03, bone2Name=dyn_chain_02, swing1LimitDegrees=20, swing2LimitDegrees=20, twistLimitDegrees=45, twistMotion=Locked}
 {
  bone1Name=base_spike_01
  bone2Name=pack
  swing1LimitDegrees=45
  swing1Motion=Free
  swing2LimitDegrees=45
  swing2Motion=Free
  twistLimitDegrees=45
  twistMotion=Free
 }
 {
  bone1Name=dyn_spike_01
  bone2Name=base_spike_01
  swing1LimitDegrees=2.5
  swing2LimitDegrees=10
  swing2Motion=Locked
  twistLimitDegrees=45
  twistMotion=Locked
 }
 {
  bone1Name=base_spike_02
  bone2Name=pack
  swing1LimitDegrees=45
  swing1Motion=Free
  swing2LimitDegrees=45
  swing2Motion=Free
  twistLimitDegrees=45
  twistMotion=Free
 }
 {
  bone1Name=dyn_spike_02
  bone2Name=base_spike_02
  swing1LimitDegrees=20
  swing2LimitDegrees=10
  swing2Motion=Locked
  twistLimitDegrees=45
  twistMotion=Locked
 }
 {
  bone1Name=base_ear_l
  bone2Name=pack
  swing1LimitDegrees=45
  swing1Motion=Free
  swing2LimitDegrees=45
  swing2Motion=Free
  twistLimitDegrees=45
  twistMotion=Free
 }
 {bone1Name=dyn_ear_l_01, bone2Name=base_ear_l, swing1LimitDegrees=10, swing2LimitDegrees=10, twistLimitDegrees=45, twistMotion=Locked}
 {bone1Name=dyn_ear_l_02, bone2Name=base_ear_l, swing1LimitDegrees=20, swing2LimitDegrees=20, twistLimitDegrees=45, twistMotion=Locked}
 {
  bone1Name=base_ear_r
  bone2Name=pack
  swing1LimitDegrees=45
  swing1Motion=Free
  swing2LimitDegrees=45
  swing2Motion=Free
  twistLimitDegrees=45
  twistMotion=Free
 }
 {bone1Name=dyn_ear_r_01, bone2Name=base_ear_r, swing1LimitDegrees=20, swing2LimitDegrees=30, twistLimitDegrees=45, twistMotion=Locked}
 {bone1Name=dyn_ear_r_02, bone2Name=base_ear_r, swing1LimitDegrees=20, swing2LimitDegrees=25, twistLimitDegrees=45, twistMotion=Locked}
]
